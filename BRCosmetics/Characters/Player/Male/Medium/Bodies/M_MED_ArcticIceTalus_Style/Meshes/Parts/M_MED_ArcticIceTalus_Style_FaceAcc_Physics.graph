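# Physics asset for the face accessory of the male medium ArcticIceTalus_Style player cosmetic, covering a left and right pair of antennae.
# Each antenna is a chain from a small sphere at the base bone through capsules on the parent, 01 and 02 bones.
# Constraints link neighboring bones with tight, uniform swing and twist limits, keeping each antenna nearly rigid.
# No constraint joins either base to the separate head-offset sphere or to the other antenna.
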PhysicsAsset /BRCosmetics/Characters/Player/Male/Medium/Bodies/M_MED_ArcticIceTalus_Style/Meshes/Parts/M_MED_ArcticIceTalus_Style_FaceAcc_Physics
bodies {
 dyn_antenna_01_l [{shapeType=Capsule, length=2.0211, radius=0.7579, shapeName=dyn_antenna_01_l_capsule, center=<1.8123 0 0>, rotation=<rot 90 0 0>}]
 dyn_antenna_01_r [{shapeType=Capsule, length=2.0264, radius=0.7584, shapeName=dyn_antenna_01_r_capsule, center=<1.8123 0 0>, rotation=<rot 90 0 0>}]
 dyn_antenna_02_l [{shapeType=Capsule, length=2.2242, radius=0.7519, shapeName=dyn_antenna_02_l_capsule, center=<1.8123 0 0>, rotation=<rot 90 0 0>}]
 dyn_antenna_02_r [{shapeType=Capsule, length=2.2276, radius=0.7599, shapeName=dyn_antenna_02_r_capsule, center=<1.8123 0 0>, rotation=<rot 90 0 0>}]
 dyn_antenna_base_l [{shapeType=Sphere, radius=0.5, shapeName=dyn_antenna_base_l_sphere}]
 dyn_antenna_base_r [{shapeType=Sphere, radius=0.5, shapeName=dyn_antenna_base_r_sphere}]
 dyn_antenna_parent_l [{shapeType=Capsule, length=2.0957, radius=0.985, shapeName=dyn_antenna_parent_l_capsule, center=<1.8123 0 0>, rotation=<rot 90 0 0>}]
 dyn_antenna_parent_r [{shapeType=Capsule, length=2.1015, radius=0.985, shapeName=dyn_antenna_parent_r_capsule, center=<1.8123 0 0>, rotation=<rot 90 0 0>}]
 dyn_head_offset [{shapeType=Sphere, radius=0.5, shapeName=dyn_head_offset_sphere}]
}
constraints [
 {bone1Name=dyn_antenna_02_r, bone2Name=dyn_antenna_01_r, swing1LimitDegrees=2, swing2LimitDegrees=2, twistLimitDegrees=2}
 {bone1Name=dyn_antenna_01_r, bone2Name=dyn_antenna_parent_r, swing1LimitDegrees=2, swing2LimitDegrees=2, twistLimitDegrees=2}
 {bone1Name=dyn_antenna_parent_r, bone2Name=dyn_antenna_base_r, swing1LimitDegrees=2, swing2LimitDegrees=2, twistLimitDegrees=2}
 {bone1Name=dyn_antenna_02_l, bone2Name=dyn_antenna_01_l, swing1LimitDegrees=2, swing2LimitDegrees=2, twistLimitDegrees=2}
 {bone1Name=dyn_antenna_01_l, bone2Name=dyn_antenna_parent_l, swing1LimitDegrees=2, swing2LimitDegrees=2, twistLimitDegrees=2}
 {bone1Name=dyn_antenna_parent_l, bone2Name=dyn_antenna_base_l, swing1LimitDegrees=2, swing2LimitDegrees=2, twistLimitDegrees=2}
]
