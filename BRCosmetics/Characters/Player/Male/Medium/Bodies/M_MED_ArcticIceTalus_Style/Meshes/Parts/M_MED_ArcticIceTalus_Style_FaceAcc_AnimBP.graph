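# Animation Blueprint for the face accessory mesh of the male medium ArcticIceTalus Style character, adding physics-driven secondary motion.
# It copies the pose from the attached parent mesh, then runs a RigidBody node with its own physics asset before output.
# Simulation runs in head bone space with overridden gravity, scaled-down component motion and capped velocities and accelerations to keep the motion restrained.
# The event graph has no logic.
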AnimBlueprint /BRCosmetics/Characters/Player/Male/Medium/Bodies/M_MED_ArcticIceTalus_Style/Meshes/Parts/M_MED_ArcticIceTalus_Style_FaceAcc_AnimBP
target_skeleton=&/M_MED_ArcticIceTalus_Style_FaceAcc_Skeleton
graphs [
 AnimGraph {nodes [
   n1 = RigidBody(ComponentPose=Local To Component(LocalPose=Copy Pose From Mesh() {bUseAttachedParent=true})) {
    bClampLinearTranslationLimitToRefPose=true
    bOverrideWorldGravity=true
    overridePhysicsAsset=&/M_MED_ArcticIceTalus_Style_FaceAcc_Physics
    simulationSpace=BaseBoneSpace
    alphaBoolBlend {blendOption=Linear}
    baseBoneRef {boneName=head}
    componentAppliedLinearAccClamp=<6000 6000 6000>
    componentLinearAccScale=<0.1 0.1 0.1>
    componentLinearVelScale=<0.1 0.1 0.1>
    overrideWorldGravity {z=-980}
    simSpaceSettings {
     maxAngularAcceleration=500
     maxAngularVelocity=3
     maxLinearAcceleration=400
     maxLinearVelocity=250
     velocityScaleZ=0
     worldAlpha=1
     externalLinearDragV=<0.1 0.1>
    }
   }
   Output Pose(Result=Component To Local(ComponentPose=@n1)) {layerGroup=None}
  ]}
 EventGraph {nodes [Event BlueprintUpdateAnimation()]}
]
variables []
